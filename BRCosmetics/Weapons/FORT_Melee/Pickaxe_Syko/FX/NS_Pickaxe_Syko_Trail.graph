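# Niagara system for the Syko pickaxe's swing trail, built from two animation-trail ribbon emitters with two ribbon renderers each.
# Both spawn short-lived ribbons through the FN anim-trail modules, dim the initial color by time of day, and fade color and alpha over life.
# Main_Trail randomizes material parameters and uses a variant color selector holding blue and orange gradients.
# The shorter Sub_Trail uses a pale gradient and animates material parameters by curve.
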NiagaraSystem /BRCosmetics/Weapons/FORT_Melee/Pickaxe_Syko/FX/NS_Pickaxe_Syko_Trail
dependencies {
 usedDataInterfaces [&NiagaraDataInterfaceColorCurve, &NiagaraDataInterfaceCurve, &NiagaraDataInterfaceSocketReader]
 usedDynamicInputs [&SetBoolByIntComparison, &MakeIntFromFloat]
 usedModules [
  &FN_VariantManager
  &self:SystemSpawnScript.SystemScriptSource.SystemScriptGraph.NiagaraNodeAssignment_0.SetVariables_#1
  &self:SystemSpawnScript.SystemScriptSource.SystemScriptGraph.NiagaraNodeAssignment_2.SetVariables_#2
  &SystemState
 ]
}
emitters {
 Main_Trail {
  emitterSpawnScript [&FN_RibbonIDOffsetUtility()]
  emitterUpdateScript [
   &EmitterState()
   &FN_SpawnAnimTrail()
  ]
  particleSpawnScript [
   &FN_InitializeAnimTrailBoneIndecies()
   &FN_InitializeAnimTrail(
    Color=&Multiply_LinearColorByFloat(
     Linear Color=<rgba 1 1 1 1>
     Float=&FNTimeOfDayScalar(Night Scale=0.4, Morning Scale=0.6, Evening Scale=0.4)
    )
    Lifetime=0.3
   )
   &DynamicMaterialParameters(
    Index 0 Param 1=&RandomRangeFloat(Minimum=0, Maximum=1)
    Index 0 Param 2=&RandomRangeFloat(Minimum=0, Maximum=1)
    Index 0 Param 3=1
    Index 0 Param 4=1
    Module.Index0Float0Write=true
    Module.Index0Float1Write=true
    Module.Index0Float2Write=true
    Module.Index0Float3Write=true
   )
  ]
  particleUpdateScript [
   &SolveForcesAndVelocity(Clamp Velocity=false, Limit Acceleration=false)
   &FN_UpdateAnimTrail() {module=FN_UpdateAnimTrail001}
   &ScaleColor(
    Scale Alpha=&FloatFromCurve(FloatCurve=DataInterface() {propertyValues {Curve {keys [
         {arriveTangent=-2.6867, interpMode=RCIM_Cubic, time=0, value=1}
         {arriveTangent=-3.0842, interpMode=RCIM_Cubic, leaveTangent=-3.0842, tangentMode=RCTM_User, time=0.0747, value=0.8886}
         {arriveTangent=-0.4677, interpMode=RCIM_Cubic, leaveTangent=-0.4677, tangentMode=RCTM_User, time=0.36, value=0.1281}
         {arriveTangentWeight=0.2133, interpMode=RCIM_Cubic, time=1, value=0}
        ]}}})
    Scale RGB=&MakeVectorFromLinearColorRGB(Color=&VariantColorSelector(
      Variant Amount=0
      Original Color=&ColorFromCurve(ColorCurve=DataInterface() {propertyValues {
         AlphaCurve {keys [<at 0 1 RCIM_Cubic>]}
         BlueCurve {keys [<at 0 5 RCIM_Cubic>,<at 0.111 2>,<at 0.42 0.6376>]}
         GreenCurve {keys [<at 0 2.9821 RCIM_Cubic>,<at 0.111 0.536>,<at 0.42 0.1384>]}
         RedCurve {keys [<at 0 0.6432 RCIM_Cubic>,<at 0.111 0.229>,<at 0.42 0.1329>]}
        }})
      Variant 1 Color=&ColorFromCurve(ColorCurve=DataInterface() {propertyValues {
         AlphaCurve {keys [<at 0 1 RCIM_Cubic>]}
         BlueCurve {keys [<at 0 1.0716 RCIM_Cubic>,<at 0.111 0.268>,<at 0.42 0.009>]}
         GreenCurve {keys [<at 0 1.4702 RCIM_Cubic>,<at 0.111 0.5>,<at 0.42 0>]}
         RedCurve {keys [<at 0 5 RCIM_Cubic>,<at 0.111 1.55>,<at 0.42 0.1771>]}
        }})
      Variant 2 Color=?
     ))
    ScaleA=true
    ScaleRGB=true
   )
  ]
  rendererClasses [&NiagaraRibbonRendererProperties]
  renderers [
   {rendererClass=&NiagaraRibbonRendererProperties}
   {rendererClass=&NiagaraRibbonRendererProperties, rendererIndex=1}
  ]
 }
 Sub_Trail {
  emitterSpawnScript [&FN_RibbonIDOffsetUtility()]
  emitterUpdateScript [
   &EmitterState()
   &FN_SpawnAnimTrail()
  ]
  particleSpawnScript [
   &FN_InitializeAnimTrailBoneIndecies()
   &FN_InitializeAnimTrail(
    Color=&Multiply_LinearColorByFloat(
     Linear Color=<rgba 1 1 1 1>
     Float=&FNTimeOfDayScalar(Night Scale=0.35, Morning Scale=0.55, Evening Scale=0.35)
    )
    Lifetime=0.22
   )
  ]
  particleUpdateScript [
   &SolveForcesAndVelocity(Clamp Velocity=false, Limit Acceleration=false)
   &FN_UpdateAnimTrail() {module=FN_UpdateAnimTrail001}
   &ScaleColor(
    Scale Alpha=&FloatFromCurve(FloatCurve=DataInterface() {propertyValues {Curve {keys [
         {arriveTangent=-2.6867, interpMode=RCIM_Cubic, time=0, value=1}
         {arriveTangent=-3.1483, interpMode=RCIM_Cubic, leaveTangent=-3.1483, tangentMode=RCTM_User, time=1, value=0}
        ]}}})
    Scale RGB=&MakeVectorFromLinearColorRGB(Color=&ColorFromCurve(ColorCurve=DataInterface() {propertyValues {
        AlphaCurve {keys [<at 0 1 RCIM_Cubic>]}
        BlueCurve {keys [<at 0 3.789 RCIM_Cubic>,<at 0.111 0.3763>,<at 1 0.026>]}
        GreenCurve {keys [<at 0 5 RCIM_Cubic>,<at 0.111 0.6376>,<at 1 0.015>]}
        RedCurve {keys [<at 0 4.908 RCIM_Cubic>,<at 0.111 0.7835>,<at 1 0.0521>]}
       }}))
    ScaleA=true
    ScaleRGB=true
   )
   &DynamicMaterialParameters(
    Index 0 Param 1=&FloatFromCurve(FloatCurve=DataInterface() {propertyValues {Curve {keys [
         <at 0.0184 0.1554 RCIM_Cubic>
         {arriveTangent=0.6823, interpMode=RCIM_Cubic, leaveTangent=0.6823, time=0.476, value=0.2822}
         <at 0.8203 0.7026 RCIM_Cubic>
        ]}}})
    Index 0 Param 2=&FloatFromCurve(FloatCurve=DataInterface() {propertyValues {Curve {keys [<at 0.4141 0.5787>,<at 0.8973 0.657>]}}})
    Index 0 Param 3=&FloatFromCurve(FloatCurve=DataInterface() {propertyValues {Curve {keys [
         {arriveTangent=-1.9013, interpMode=RCIM_Cubic, leaveTangent=-1.9013, tangentMode=RCTM_User, time=0, value=1}
         {arriveTangent=-0.1968, tangentMode=RCTM_User, time=1, value=0}
        ]}}})
    Index 0 Param 4=1
    Module.Index0Float0Write=true
    Module.Index0Float1Write=true
    Module.Index0Float2Write=true
    Module.Index0Float3Write=true
   )
  ]
  rendererClasses [&NiagaraRibbonRendererProperties]
  renderers [
   {rendererClass=&NiagaraRibbonRendererProperties}
   {rendererClass=&NiagaraRibbonRendererProperties, rendererIndex=1}
  ]
 }
}
summary {
 systemName=NS_Pickaxe_Syko_Trail
 emitters [
  {emitterName=Main_Trail, rendererClasses [&NiagaraRibbonRendererProperties]}
  {emitterName=Sub_Trail, rendererClasses [&NiagaraRibbonRendererProperties]}
 ]
 userVariables [
  User.RibbonIDOffset {type=NiagaraInt32}
  User.Variant {type=NiagaraFloat, defaultValue=1, description=""}
 ]
}
