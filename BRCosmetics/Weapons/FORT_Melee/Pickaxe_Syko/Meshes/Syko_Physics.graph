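# Physics asset for the Syko pickaxe skeletal mesh, a two-handed melee weapon cosmetic.
# It holds three capsule bodies: a small one on the handle bone and two identical larger, rotated ones on the dyn_float_a and dyn_float_b bones.
# Constraints chain dyn_float_b to dyn_float_a and dyn_float_a to the handle, with swing and twist locked on both joints.
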PhysicsAsset /BRCosmetics/Weapons/FORT_Melee/Pickaxe_Syko/Meshes/Syko_Physics
bodies {
 Melee_TwoHanded_Handle [{shapeType=Capsule, length=1, radius=1}]
 dyn_float_a [{shapeType=Capsule, length=3.75, radius=4.5, rotation=<rot 90 0 0>}]
 dyn_float_b [{shapeType=Capsule, length=3.75, radius=4.5, rotation=<rot 90 0 0>}]
}
constraints [
 {
  bone1Name=dyn_float_a
  bone2Name=Melee_TwoHanded_Handle
  swing1LimitDegrees=45
  swing1Motion=Locked
  swing2LimitDegrees=45
  swing2Motion=Locked
  twistLimitDegrees=45
  twistMotion=Locked
 }
 {
  bone1Name=dyn_float_b
  bone2Name=dyn_float_a
  swing1LimitDegrees=45
  swing1Motion=Locked
  swing2LimitDegrees=45
  swing2Motion=Locked
  twistLimitDegrees=45
  twistMotion=Locked
 }
]
