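# Animation Blueprint for the Syko pickaxe skeleton that makes two bones, dyn_float_a and dyn_float_b, bob procedurally.
# Each update, RandomMoveFNC scales a sine of game time by frequency and amplitude to fill two offset vectors, with negative amplitudes for the second.
# Additive bone-space Transform Bone nodes apply these offsets to the input pose.
# AnimDynamics, RigidBody and Perlin noise nodes are present but do not feed the output.
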AnimBlueprint /BRCosmetics/Weapons/FORT_Melee/Pickaxe_Syko/Meshes/Syko_AnimBP
target_skeleton=&/Syko_Skeleton
graphs [
 AnimGraph {nodes [
   Anim Dynamics - Bone: dyn_float_a(Alpha=1, ExternalForce="0,0,0") {
    bGravityOverrideInSimSpace=true
    bLinearSpring=true
    bOverrideLinearDamping=true
    gravityScale=0.35
    linearDampingOverride=0.9999
    linearSpringConstant=300
    alphaBoolBlend {blendOption=Linear}
    boundBone {boneName=dyn_float_a}
    componentLinearAccScale {y=-0.5}
    componentLinearVelScale {y=0.5}
    physicsBodyDefinitions [{
      sphereCollisionRadius=0
      boundBone {boneName=dyn_float_a}
      boxExtents=<10 10 10>
      constraintSetup {angularLimitsMax=<0 0 0>, angularLimitsMin=<0 0 0>, angularTarget=<0 0 0>, linearAxesMax=<0.5 0 0>, linearAxesMin=<-0.5 0 0>}
      localJointOffset=<-1 0 0>
     }]
   }
   RigidBody() {
    bClampLinearTranslationLimitToRefPose=true
    bOverrideWorldGravity=true
    overridePhysicsAsset=&/Syko_Physics
    alphaBoolBlend {blendOption=Linear}
    baseBoneRef {boneName=Melee_TwoHanded_Handle}
    componentLinearAccScale=<0.1 0.1 0.1>
    componentLinearVelScale=<0.1 0.1 0.1>
   }
   AnimGraphNode_SubInput_0 = Input Pose()
   n1 = "Transform (Modify) Bone - Bone: dyn_float_a"(
    ComponentPose=Local To Component(LocalPose=@AnimGraphNode_SubInput_0)
    Alpha=1
    Translation=Get floatingVec1() -> (floatingVec1:Vector)
    Rotation="0,0,0"
    Scale="1,1,1"
   ) {
    rotationMode=BMM_Additive
    rotationSpace=BCS_BoneSpace
    translationMode=BMM_Additive
    translationSpace=BCS_BoneSpace
    alphaBoolBlend {blendOption=Linear}
    boneToModify {boneName=dyn_float_a}
   }
   n2 = "Transform (Modify) Bone - Bone: dyn_float_b"(ComponentPose=@n1, Alpha=1, Translation=Get floatingVec2() -> (floatingVec2:Vector), Rotation="0,0,0", Scale="1,1,1") {
    rotationMode=BMM_Additive
    rotationSpace=BCS_BoneSpace
    translationMode=BMM_Additive
    translationSpace=BCS_BoneSpace
    alphaBoolBlend {blendOption=Linear}
    boneToModify {boneName=dyn_float_b}
   }
   Output Pose(Result=Component To Local(ComponentPose=@n2)) {layerGroup=None}
  ]}
 EventGraph {nodes [
   Event BlueprintUpdateAnimation()
   > n1 = Sequence!(execute) -> (then_0, then_1, then_2:Exec, then_3:Exec)
   > n2 = RandomMoveFNC!(execute, self, frequency:double=2, amp:double=0.5) -> (then, tran_output:double)
   n3 = RandomMoveFNC(execute=@n1.then_1, frequency=2, amp=0.2)
   Set floatingVec1(execute=[@n3.then, @n2.then], floatingVec1:Vector=MakeVector(X=@n2, Y=@n3, Z=@n3)) -> (Output_Get:Vector)
   n4 = RandomMoveFNC(execute=@n1.then_2, frequency=2.5, amp=-0.45)
   n5 = RandomMoveFNC(execute=@n1.then_3, frequency=2.5, amp=-0.2)
   Set floatingVec2(execute=[@n5.then, @n4.then], floatingVec2:Vector=MakeVector(X=@n4, Y=@n5, Z=@n5)) -> (Output_Get:Vector)
   TryGetPawnOwner()
  ]}
 RandomMoveFNC {nodes [
   n1 = RandomMoveFNC!() -> (then, frequency:double=~, amp:double=~) {is=title K2Node_FunctionEntry}
   n2 = float * float(A=@n1, B=GetGameTimeInSeconds())
   n3 = float * float(A="Sin (Radians)"(A=@n2), B=@n1.amp)
   > Return Node!(execute, tran_output:double=@n3)
   PerlinNoise1D!(Value:float=@n2) -> (ReturnValue:float)
   PerlinNoise1D(Value=@n3)
  ]}
]
variables [
 floatingVec1 {type=Vector}
 floatingVec2 {type=Vector}
]
